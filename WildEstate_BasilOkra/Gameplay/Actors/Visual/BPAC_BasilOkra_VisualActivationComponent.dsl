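; Actor component for BasilOkra visual actors that registers its owner with the BasilOkra console.
; On BeginPlay it finds the BGA_BasilOkra_Console actor and, if valid, adds the owner to that console's RegisteredVisualActors array.
; A helper event, SetActiveOnVFXChildActor, casts a child actor component's actor to BP_BasilOkra_FX and sets its FX component active or inactive without a reset.
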(event EventBeginPlay
  (CallFunction|RegisterOwningActor))

(event EventTick
  (params (DeltaSeconds :type Float%20%28single-precision%29)))

(event Custom|RegisterOwningActor
  (bind _returnvalue (Actor|GetActorOfClass "/WildEstate_BasilOkra/Gameplay/Actors/BGA_BasilOkra_Console.BGA_BasilOkra_Console_C"))
  (Utilities|IsValid _returnvalue
    (:Is%20Valid
      (Utilities|Array|Add (Variables|/WildEstate_BasilOkra/Gameplay/Actors/BGA_BasilOkra_Console.BGA_BasilOkra_Console_C|GetRegisteredVisualActorsArray _returnvalue) (Components|GetOwner)))))

(event Custom|SetActiveOnVFXChildActor
  (params (ChildActorComponent :type Child%20Actor%20Component%20Object%20Reference) (IsActive :type Boolean))
  (bind _asbp_basil_okra_fx (Utilities|Casting|CastToBP_BasilOkra_FX (Variables|/Script/Engine.ChildActorComponent|GetChildActor ChildActorComponent)))
  (Components|Activation|SetActive (Variables|/WildEstate_BasilOkra/Gameplay/Actors/Visual/BP_BasilOkra_FX.BP_BasilOkra_FX_C|GetFX _asbp_basil_okra_fx) IsActive false))
