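; Blueprint interface for visual actors in the WildEstate BasilOkra gameplay content, letting callers switch an actor's visuals on and off.
; It declares two functions, ActivateVisuals and DeactivateVisuals, with no parameters or return values.
; The interface holds no logic, so each implementing Blueprint defines its own activation and deactivation behavior.
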(fn ActivateVisuals)

(fn DeactivateVisuals)
