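; Visual Blueprint for an electrical tower that toggles child-actor spark effects through its visual activation component.
; On activate or deactivate, it loops over the children of FXParent, casts them to child actor components and has the component set each active or inactive.
; Activation also calls a custom event inside the loop that pushes one FX scale variable into the Niagara FXScale parameter of four hard-referenced spark actors, so it repeats per child.
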(event Custom|SetChildActorFXScale
  (bind _fx_scale__value (Variables|Default|GetFX%20Scale%20%20Value))
  (Niagara|SetNiagaraVariable(Float) (Variables|/WildEstate_BasilOkra/Gameplay/Actors/Visual/BP_BasilOkra_FX.BP_BasilOkra_FX_C|GetFX (Utilities|Casting|CastToBP_BasilOkra_FX_Sparks (Variables|/Script/Engine.ChildActorComponent|GetChildActor (Variables|Default|GetBP_BasilOkra_FX_Sparks)))) "FXScale" _fx_scale__value)
  (Niagara|SetNiagaraVariable(Float) (Variables|/WildEstate_BasilOkra/Gameplay/Actors/Visual/BP_BasilOkra_FX.BP_BasilOkra_FX_C|GetFX (Utilities|Casting|CastToBP_BasilOkra_FX_Sparks (Variables|/Script/Engine.ChildActorComponent|GetChildActor (Variables|Default|GetBP_BasilOkra_FX_Sparks1)))) "FXScale" _fx_scale__value)
  (Niagara|SetNiagaraVariable(Float) (Variables|/WildEstate_BasilOkra/Gameplay/Actors/Visual/BP_BasilOkra_FX.BP_BasilOkra_FX_C|GetFX (Utilities|Casting|CastToBP_BasilOkra_FX_Sparks (Variables|/Script/Engine.ChildActorComponent|GetChildActor (Variables|Default|GetBP_BasilOkra_FX_Sparks2)))) "FXScale" _fx_scale__value)
  (Niagara|SetNiagaraVariable(Float) (Variables|/WildEstate_BasilOkra/Gameplay/Actors/Visual/BP_BasilOkra_FX.BP_BasilOkra_FX_C|GetFX (Utilities|Casting|CastToBP_BasilOkra_FX_Sparks (Variables|/Script/Engine.ChildActorComponent|GetChildActor (Variables|Default|GetBP_BasilOkra_FX_Sparks3)))) "FXScale" _fx_scale__value))

(event Collision|EventActorBeginOverlap
  (params (OtherActor :type Actor%20Object%20Reference))
  (|CallParentFunction OtherActor) ; This node is disabled and will not be called.\nDrag off pins to build functionality.
)

(event EventTick
  (params (DeltaSeconds :type Float%20%28single-precision%29))
  (|CallParentFunction DeltaSeconds) ; This node is disabled and will not be called.\nDrag off pins to build functionality.
)

(event ActivateVisuals(BPAC_BasilOkra_VisualActivationComponent)
  (for _array_element (Components|GetChildrenComponents (Variables|Default|GetFXParent) false)
    (bind _aschild_actor_component (Utilities|Casting|CastToChildActorComponent _array_element))
    (Class|BPACBasilOkraVisualActivationComponent|SetActiveonVFXChildActor (Variables|Default|GetBPAC_BasilOkra_VisualActivationComponent) _aschild_actor_component true)
    (CallFunction|SetChildActorFXScale)))

(event DeactivateVisuals(BPAC_BasilOkra_VisualActivationComponent)
  (for _array_element (Components|GetChildrenComponents (Variables|Default|GetFXParent) false)
    (bind _aschild_actor_component (Utilities|Casting|CastToChildActorComponent _array_element))
    (Class|BPACBasilOkraVisualActivationComponent|SetActiveonVFXChildActor (Variables|Default|GetBPAC_BasilOkra_VisualActivationComponent) _aschild_actor_component false)))

(fn ConstructionScript
  (|CallParentFunction))
