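; Blueprint visual actor for the BasilOkra exterior "Pair_Str_512" piece in WildEstate, with two steam effects that switch on and off.
; Activate and deactivate events from its visual activation component call that component's SetActiveonVFXChildActor on both steam FX members.
; BeginPlay, overlap and Tick events are empty, and the construction script only calls its parent.
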(event EventBeginPlay)

(event Collision|EventActorBeginOverlap
  (params (OtherActor :type Actor%20Object%20Reference)))

(event EventTick
  (params (DeltaSeconds :type Float%20%28single-precision%29)))

(event ActivateVisuals(BPAC_BasilOkra_VisualActivationComponent)
  (Class|BPACBasilOkraVisualActivationComponent|SetActiveonVFXChildActor (Variables|Default|GetBPAC_BasilOkra_VisualActivationComponent) (Variables|Default|GetBP_BasilOkra_FX_Steam) true)
  (Class|BPACBasilOkraVisualActivationComponent|SetActiveonVFXChildActor (Variables|Default|GetBPAC_BasilOkra_VisualActivationComponent) (Variables|Default|GetBP_BasilOkra_FX_Steam1) true))

(event DeactivateVisuals(BPAC_BasilOkra_VisualActivationComponent)
  (Class|BPACBasilOkraVisualActivationComponent|SetActiveonVFXChildActor (Variables|Default|GetBPAC_BasilOkra_VisualActivationComponent) (Variables|Default|GetBP_BasilOkra_FX_Steam) false)
  (Class|BPACBasilOkraVisualActivationComponent|SetActiveonVFXChildActor (Variables|Default|GetBPAC_BasilOkra_VisualActivationComponent) (Variables|Default|GetBP_BasilOkra_FX_Steam1) false))

(fn ConstructionScript
  (|CallParentFunction))
