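; BP_BasilOkra_GlowDonut_Wall_A_A is a wall visual actor that toggles a GlowDonut cage light alarm.
; It handles activate and deactivate events from a BasilOkra visual activation component.
; Each event casts the Alarm child actor to BP_GlowDonut_CageLight_A_BasilOkra and calls its activate or deactivate alarm function.
; Begin play, overlap and tick are unused disabled stubs.
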(event EventBeginPlay
  (|CallParentFunction) ; This node is disabled and will not be called.\nDrag off pins to build functionality.
)

(event Collision|EventActorBeginOverlap
  (params (OtherActor :type Actor%20Object%20Reference))
  (|CallParentFunction OtherActor) ; This node is disabled and will not be called.\nDrag off pins to build functionality.
)

(event EventTick
  (params (DeltaSeconds :type Float%20%28single-precision%29))
  (|CallParentFunction DeltaSeconds) ; This node is disabled and will not be called.\nDrag off pins to build functionality.
)

(event ActivateVisuals(BPAC_BasilOkra_VisualActivationComponent)
  (bind _asbp_glow_donut_cage_light_a_basil_okra (Utilities|Casting|CastToBP_GlowDonut_CageLight_A_BasilOkra (Variables|/Script/Engine.ChildActorComponent|GetChildActor (Variables|Default|GetAlarm))))
  (Class|BPGlowDonutCageLightABasilOkra|BasilOkraActivateAlarm _asbp_glow_donut_cage_light_a_basil_okra))

(event DeactivateVisuals(BPAC_BasilOkra_VisualActivationComponent)
  (bind _asbp_glow_donut_cage_light_a_basil_okra (Utilities|Casting|CastToBP_GlowDonut_CageLight_A_BasilOkra (Variables|/Script/Engine.ChildActorComponent|GetChildActor (Variables|Default|GetAlarm))))
  (Class|BPGlowDonutCageLightABasilOkra|BasilOkraDeactivateAlarm _asbp_glow_donut_cage_light_a_basil_okra))

(fn ConstructionScript
  (|CallParentFunction))
