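; Light post actor Blueprint for BasilOkra that toggles two spark VFX child actors on and off.
; Activate and Deactivate events from its visual activation component call that component's SetActiveonVFXChildActor on both spark effects with true or false.
; BeginPlay, overlap and Tick are disabled stubs, and the construction script only calls its parent.
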(event EventBeginPlay
  (|CallParentFunction) ; This node is disabled and will not be called.\nDrag off pins to build functionality.
)

(event Collision|EventActorBeginOverlap
  (params (OtherActor :type Actor%20Object%20Reference))
  (|CallParentFunction OtherActor) ; This node is disabled and will not be called.\nDrag off pins to build functionality.
)

(event EventTick
  (params (DeltaSeconds :type Float%20%28single-precision%29))
  (|CallParentFunction DeltaSeconds) ; This node is disabled and will not be called.\nDrag off pins to build functionality.
)

(event ActivateVisuals(BPAC_BasilOkra_VisualActivationComponent)
  (Class|BPACBasilOkraVisualActivationComponent|SetActiveonVFXChildActor (Variables|Default|GetBPAC_BasilOkra_VisualActivationComponent) (Variables|Default|GetBP_BasilOkra_FX_Sparks) true)
  (Class|BPACBasilOkraVisualActivationComponent|SetActiveonVFXChildActor (Variables|Default|GetBPAC_BasilOkra_VisualActivationComponent) (Variables|Default|GetBP_BasilOkra_FX_Sparks1) true))

(event DeactivateVisuals(BPAC_BasilOkra_VisualActivationComponent)
  (Class|BPACBasilOkraVisualActivationComponent|SetActiveonVFXChildActor (Variables|Default|GetBPAC_BasilOkra_VisualActivationComponent) (Variables|Default|GetBP_BasilOkra_FX_Sparks) false)
  (Class|BPACBasilOkraVisualActivationComponent|SetActiveonVFXChildActor (Variables|Default|GetBPAC_BasilOkra_VisualActivationComponent) (Variables|Default|GetBP_BasilOkra_FX_Sparks1) false))

(fn ConstructionScript
  (|CallParentFunction))
