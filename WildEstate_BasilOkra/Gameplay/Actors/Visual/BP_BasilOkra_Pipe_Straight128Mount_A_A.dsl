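; Straight pipe mount visual actor for the BasilOkra set, with a steam effect that switches on and off through its visual activation component.
; ActivateVisuals enables the silent steam VFX child actor and sets its Niagara FXScale parameter, while DeactivateVisuals disables it.
; When its linked actor is destroyed, the pipe calls Die on itself, forwarding the damage causer.
; BeginPlay, overlap and tick are empty, and construction only calls the parent.
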(event EventBeginPlay)

(event Collision|EventActorBeginOverlap
  (params (OtherActor :type Actor%20Object%20Reference)))

(event EventTick
  (params (DeltaSeconds :type Float%20%28single-precision%29)))

(event ActivateVisuals(BPAC_BasilOkra_VisualActivationComponent)
  (bind _bp_basilokra_fx_steam_noaudio (Variables|Default|GetBP_BasilOkra_FX_Steam_NoAudio))
  (Class|BPACBasilOkraVisualActivationComponent|SetActiveonVFXChildActor (Variables|Default|GetBPAC_BasilOkra_VisualActivationComponent) _bp_basilokra_fx_steam_noaudio true)
  (bind _asbp_basil_okra_fx_steam_no_audio (Utilities|Casting|CastToBP_BasilOkra_FX_Steam_NoAudio (Variables|/Script/Engine.ChildActorComponent|GetChildActor _bp_basilokra_fx_steam_noaudio)))
  (Niagara|SetNiagaraVariable(Float) (Variables|/WildEstate_BasilOkra/Gameplay/Actors/Visual/BP_BasilOkra_FX.BP_BasilOkra_FX_C|GetFX _asbp_basil_okra_fx_steam_no_audio) "FXScale" 1.0))

(event DeactivateVisuals(BPAC_BasilOkra_VisualActivationComponent)
  (Class|BPACBasilOkraVisualActivationComponent|SetActiveonVFXChildActor (Variables|Default|GetBPAC_BasilOkra_VisualActivationComponent) (Variables|Default|GetBP_BasilOkra_FX_Steam_NoAudio) false))

(event OnLinkedActorDestroyed(FortLinkToActor)
  (params (DamageCauser :type Actor%20Object%20Reference))
  (Gameplay|Die :self self :DamageCauser DamageCauser))

(fn ConstructionScript
  (|CallParentFunction))
