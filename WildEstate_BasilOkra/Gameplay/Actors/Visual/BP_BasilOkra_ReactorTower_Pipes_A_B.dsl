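; Reactor tower pipes visual actor for BasilOkra that toggles a steam effect and dies along with its linked actor.
; BPAC_BasilOkra_VisualActivationComponent's ActivateVisuals and DeactivateVisuals events switch the BP_BasilOkra_FX_Steam VFX child actor on and off through that same component.
; When FortLinkToActor reports the linked actor destroyed, it calls Die on itself, passing along the damage causer.
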(event OnLinkedActorDestroyed(FortLinkToActor)
  (params (DamageCauser :type Actor%20Object%20Reference))
  (Gameplay|Die :self self :DamageCauser DamageCauser))

(event ActivateVisuals(BPAC_BasilOkra_VisualActivationComponent)
  (Class|BPACBasilOkraVisualActivationComponent|SetActiveonVFXChildActor (Variables|Default|GetBPAC_BasilOkra_VisualActivationComponent) (Variables|Default|GetBP_BasilOkra_FX_Steam) true))

(event DeactivateVisuals(BPAC_BasilOkra_VisualActivationComponent)
  (Class|BPACBasilOkraVisualActivationComponent|SetActiveonVFXChildActor (Variables|Default|GetBPAC_BasilOkra_VisualActivationComponent) (Variables|Default|GetBP_BasilOkra_FX_Steam) false))

(fn ConstructionScript)
