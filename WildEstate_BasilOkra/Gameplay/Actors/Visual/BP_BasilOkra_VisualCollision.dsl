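; Visual-activation volume that switches visuals on and off for actors inside its box through BPI_BasilOkra_VisualActivation.
; Its box and pawn-only collision component start inactive, with custom events toggling both.
; Activating visuals enables collision, calls the interface on overlapping actors, stores them (those tagged BasilOkra.Test twice) and prints a debug "Hello".
; Deactivating visuals calls the interface on stored actors without clearing the array or disabling collision.
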(event EventBeginPlay
  (Components|Activation|SetActive (Variables|Default|GetPawnOnly) false false)
  (Components|Activation|SetActive (Variables|Default|GetBox) false false))

(event Custom|ActivateCollision
  (Components|Activation|SetActive (Variables|Default|GetPawnOnly) true false)
  (Components|Activation|SetActive (Variables|Default|GetBox) true false))

(event Custom|DeactivateCollision
  (Components|Activation|SetActive (Variables|Default|GetPawnOnly) false false)
  (Components|Activation|SetActive (Variables|Default|GetBox) false false))

(event Custom|ActivateVisualsThroughInterface
  (CallFunction|ActivateCollision)
  (for _array_element (Collision|GetOverlappingActors (Variables|Default|GetBox) 0 :!class /Script/Engine.PrimitiveComponent)
    (bind _asbpi_basil_okra_visual_activation (Utilities|Casting|CastToBPI_BasilOkra_VisualActivation _array_element))
    (Class|BPIBasilOkraVisualActivation|ActivateVisuals _asbpi_basil_okra_visual_activation)
    (Utilities|Array|Add (Variables|Default|GetVisualActorArray) _array_element)
    (if (Actor|ActorHasTag _array_element "BasilOkra.Test")
      (Utilities|Array|Add (Variables|Default|GetVisualActorArray) _array_element)))
  (Development|PrintString "Hello"))

(event Custom|DeactivateVisualsThroughInterface
  (for _array_element (Variables|Default|GetVisualActorArray)
    (bind _asbpi_basil_okra_visual_activation (Utilities|Casting|CastToBPI_BasilOkra_VisualActivation _array_element))
    (Class|BPIBasilOkraVisualActivation|DeactivateVisuals _asbpi_basil_okra_visual_activation)))

(fn GetPawnCollision
  (return (Variables|Default|GetPawnOnly)))

(fn ConstructionScript)
