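; BP_BasilOkra_WasherConduit_A is a conduit prop actor that shows or hides spark effects as its visual state changes.
; Activate and deactivate events from the BPAC_BasilOkra_VisualActivationComponent toggle a silent sparks child actor through that same component.
; On activation it also casts that child and sets its Niagara FXScale.
; When its linked actor is destroyed, the conduit calls Die on itself, passing along the damage causer.
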(event EventBeginPlay
  (|CallParentFunction) ; This node is disabled and will not be called.\nDrag off pins to build functionality.
)

(event Collision|EventActorBeginOverlap
  (params (OtherActor :type Actor%20Object%20Reference))
  (|CallParentFunction OtherActor) ; This node is disabled and will not be called.\nDrag off pins to build functionality.
)

(event EventTick
  (params (DeltaSeconds :type Float%20%28single-precision%29))
  (|CallParentFunction DeltaSeconds) ; This node is disabled and will not be called.\nDrag off pins to build functionality.
)

(event ActivateVisuals(BPAC_BasilOkra_VisualActivationComponent)
  (bind _bp_basilokra_fx_sparks_noaudio (Variables|Default|GetBP_BasilOkra_FX_Sparks_NoAudio))
  (Class|BPACBasilOkraVisualActivationComponent|SetActiveonVFXChildActor (Variables|Default|GetBPAC_BasilOkra_VisualActivationComponent) _bp_basilokra_fx_sparks_noaudio true)
  (bind _asbp_basil_okra_fx_sparks_no_audio (Utilities|Casting|CastToBP_BasilOkra_FX_Sparks_NoAudio (Variables|/Script/Engine.ChildActorComponent|GetChildActor _bp_basilokra_fx_sparks_noaudio)))
  (Niagara|SetNiagaraVariable(Float) (Variables|/WildEstate_BasilOkra/Gameplay/Actors/Visual/BP_BasilOkra_FX.BP_BasilOkra_FX_C|GetFX _asbp_basil_okra_fx_sparks_no_audio) "FXScale" 1.0))

(event DeactivateVisuals(BPAC_BasilOkra_VisualActivationComponent)
  (Class|BPACBasilOkraVisualActivationComponent|SetActiveonVFXChildActor (Variables|Default|GetBPAC_BasilOkra_VisualActivationComponent) (Variables|Default|GetBP_BasilOkra_FX_Sparks_NoAudio) false))

(event OnLinkedActorDestroyed(FortLinkToActor)
  (params (DamageCauser :type Actor%20Object%20Reference))
  (Gameplay|Die :self self :DamageCauser DamageCauser))

(fn ConstructionScript
  (|CallParentFunction))
